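# Control Rig for the RoboKitty pet skeleton that reduces upper-arm yaw on the incoming pose.
# Each upper arm has its own Begin Execution chain that reads the local transform, converts rotation to a rotator and scales yaw to 80%.
# The result is written back with pitch, roll, translation and scale intact, and propagation to children off.
# The hierarchy lists only bones, including ear, tail, glasses and IK bones, with no controls.
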ControlRigBlueprint /BRCosmetics/Characters/Pets/Robo_kitty/Meshes/M_MED_RoboKitty_ControlRig
graphs [
 RigVMFunctionLibrary {links [], nodes []}
 RigVMModel {nodes [
   n1 = &RigUnit_BeginExecution()
   n2 = &RigUnit_GetBoneTransform(Bone=upperarm_r, Space=LocalSpace)
   n3 = /Script/RigVM.RigVMFunction_MathQuaternionToRotator(Value=@n2.Transform.Rotation) {is=struct RigVMUnitNode, template="Cast::Execute(in Value,out Result)"}
   /Script/ControlRig.RigUnit_SetBoneTransform(
    Bone=upperarm_r
    Space=LocalSpace
    Transform=(Rotation=(X=0,Y=0,Z=0,W=1),Translation=(X=0,Y=0,Z=0),Scale3D=(X=1,Y=1,Z=1))
    Weight=1
    bPropagateToChildren=False
    ExecutePin=@n1.ExecutePin
    Transform.Scale3D=@n2.Transform.Scale3D
    Transform.Translation=@n2.Transform.Translation
    Transform.Rotation=/Script/RigVM.RigVMFunction_MathQuaternionFromRotator(
     Rotator=(Pitch=0,Yaw=0,Roll=0)
     Rotator.Yaw=&RigVMFunction_MathFloatMul(B=0.8, A=@n3.Result.Yaw) {template="Multiply::Execute(in A,in B,out Result)"}.Result
     Rotator.Pitch=@n3.Result.Pitch
     Rotator.Roll=@n3.Result.Roll
    ) {is=struct RigVMUnitNode}.Result
   ) {is=struct RigVMUnitNode}
   n4 = &RigUnit_BeginExecution()
   n5 = &RigUnit_GetBoneTransform(Bone=upperarm_l, Space=LocalSpace)
   n6 = &RigVMFunction_MathQuaternionToRotator(Value=@n5.Transform.Rotation) {template="Cast::Execute(in Value,out Result)"}
   &RigUnit_SetBoneTransform(
    Bone=upperarm_l
    Space=LocalSpace
    Transform=(Rotation=(X=0,Y=0,Z=0,W=1),Translation=(X=0,Y=0,Z=0),Scale3D=(X=1,Y=1,Z=1))
    Weight=1
    bPropagateToChildren=False
    ExecutePin=@n4.ExecutePin
    Transform.Scale3D=@n5.Transform.Scale3D
    Transform.Translation=@n5.Transform.Translation
    Transform.Rotation=&RigVMFunction_MathQuaternionFromRotator(
     Rotator=(Pitch=0,Yaw=0,Roll=0)
     Rotator.Yaw=&RigVMFunction_MathFloatMul(B=0.8, A=@n6.Result.Yaw) {template="Multiply::Execute(in A,in B,out Result)"}.Result
     Rotator.Pitch=@n6.Result.Pitch
     Rotator.Roll=@n6.Result.Roll
    ).Result
   )
  ]}
]
hierarchy [
 root {loc=<0 0 0>, rot=<0 0 0 1>}
 pelvis {parent=root, loc=<0 0.2632 11.75>, rot=<0.0013 0.7071 -0.0013 -0.7071>}
 spine_01 {parent=pelvis, loc=<-0.0374 0.429 0>, rot=<0 0 -0.0624 -0.9981>}
 spine_02 {parent=spine_01, loc=<0.3851 2.7484 0>, rot=<0 0 0.0822 -0.9966>}
 spine_03 {parent=spine_02, loc=<-0.0555 3.1963 0>, rot=<0 0 -0.6154 -0.7882>}
 clavicle_l {parent=spine_03, loc=<1.8339 1.3565 -2.5278>, rot=<-0.4283 -0.5627 0.4283 -0.5627>}
 upperarm_l {parent=clavicle_l, loc=<0.084 0.5739 -2.1864>, rot=<-0.0233 -0.7167 -0.0212 -0.6967>}
 lowerarm_l {parent=upperarm_l, loc=<4.4314 0 0>, rot=<0 0 -0.2676 -0.9635>}
 hand_l {parent=lowerarm_l, loc=<4.1558 0 0>, rot=<-0.6837 0.1805 0.1805 -0.6837>}
 index_01_l {parent=hand_l, loc=<1.1049 -0.0942 -0.2046>, rot=<-0.0643 -0.678 -0.0288 -0.7317>}
 index_02_l {parent=index_01_l, loc=<0.5398 0.008 -0.0104>, rot=<0.0001 -0.007 0.0155 -0.9999>}
 index_03_l {parent=index_02_l, loc=<0.3934 -0.0205 0.0159>, rot=<-0.0006 0.007 0.0543 -0.9985>}
 lowerarm_twist_01_l {parent=lowerarm_l, loc=<1.7662 0 0>, rot=<0 0 0 -1>}
 upperarm_twist_01_l {parent=upperarm_l, loc=<0.4541 0 0>, rot=<0 0 0 1>}
 clavicle_r {parent=spine_03, loc=<1.8339 1.3565 2.5278>, rot=<0.5627 -0.4283 -0.5627 -0.4283>}
 upperarm_r {parent=clavicle_r, loc=<-0.084 -0.5739 2.1864>, rot=<-0.0233 -0.7167 -0.0212 -0.6967>}
 lowerarm_r {parent=upperarm_r, loc=<-4.4314 0 0>, rot=<0 0 -0.2676 -0.9635>}
 hand_r {parent=lowerarm_r, loc=<-4.1558 0 0>, rot=<-0.6837 0.1805 0.1805 -0.6837>}
 index_01_r {parent=hand_r, loc=<-1.1049 0.0942 0.2046>, rot=<-0.0643 -0.678 -0.0288 -0.7317>}
 index_02_r {parent=index_01_r, loc=<-0.5398 -0.008 0.0104>, rot=<0.0001 -0.007 0.0155 -0.9999>}
 index_03_r {parent=index_02_r, loc=<-0.3934 0.0205 -0.0159>, rot=<-0.0006 0.007 0.0543 -0.9985>}
 lowerarm_twist_01_r {parent=lowerarm_r, loc=<-1.7662 0 0>, rot=<0 0 0 1>}
 upperarm_twist_01_r {parent=upperarm_r, loc=<-0.4541 0 0>, rot=<0 0 0 1>}
 neck_01 {parent=spine_03, loc=<3.0689 0.6191 0>, rot=<0 0 -0.083 -0.9965>}
 collar {parent=neck_01, loc=<-0.0119 -0.0683 0>, rot=<-0.1384 -0.6934 0.1384 -0.6934>}
 neck_02 {parent=neck_01, loc=<1.623 -0.0464 0>, rot=<0 0 0.2352 -0.972>}
 head {parent=neck_02, loc=<1.5025 -0.7111 0>, rot=<0 0 0.2163 -0.9763>}
 ear_l {parent=head, loc=<6.5542 0.5165 -3.7383>, rot=<0.085 -0.843 -0.0533 -0.5285>}
 ear_r {parent=head, loc=<6.5542 0.5165 3.7383>, rot=<0.843 0.085 -0.5285 0.0533>}
 glasses {parent=head, loc=<2.7945 1.1012 0>, rot=<0.0082 -0.7071 -0.0082 -0.7071>}
 tail_01 {parent=pelvis, loc=<0.1458 -4.0091 0>, rot=<-0.0013 -0.7071 0.0013 -0.7071>}
 tail_02 {parent=tail_01, loc=<0 -2.9939 0>, rot=<0 0 0 1>}
 tail_03 {parent=tail_02, loc=<0 -2.9939 0>, rot=<0 0 0 1>}
 tail_04 {parent=tail_03, loc=<0 -2.9939 0>, rot=<0 0 0 1>}
 thigh_l {parent=pelvis, loc=<-1.0861 -3.0336 -1.9393>, rot=<0 0 0.1681 -0.9858>}
 calf_l {parent=thigh_l, loc=<-5.3369 -0.7021 0>, rot=<0 0 -0.5505 -0.8348>}
 calf_OL_l {parent=calf_l, loc=<0 0 0.0045>, rot=<-0.441 -0.5527 0.441 -0.5527>}
 calf_twist_01_l {parent=calf_l, loc=<-1.1013 0 0>, rot=<0 0 0 -1>}
 heel_l {parent=calf_l, loc=<-3.8962 -0.5585 0>, rot=<0 0 0.2539 -0.9672>}
 foot_l {parent=heel_l, loc=<-2.0123 -0.0818 0>, rot=<0 0 -0.2262 -0.9741>}
 ball_l {parent=foot_l, loc=<-0.9705 1.2315 0.0044>, rot=<0.0001 -0.0001 -0.7071 -0.7071>}
 foot_OL_l {parent=foot_l, loc=<0 0 0.0045>, rot=<0 -0.7071 0 -0.7071>}
 heel_OL_l {parent=heel_l, loc=<0 0 0.0045>, rot=<0.1687 -0.6867 -0.1687 -0.6867>}
 thigh_OL_l {parent=thigh_l, loc=<0 0 0.0045>, rot=<0 -0.7071 0 -0.7071>}
 thigh_twist_01_l {parent=thigh_l, loc=<-0.0451 0 0>, rot=<0 0 0 -1>}
 thigh_r {parent=pelvis, loc=<-1.0861 -3.0336 1.9393>, rot=<0 0 0.9858 0.1681>}
 calf_r {parent=thigh_r, loc=<5.3369 0.7021 0>, rot=<0 0 -0.5505 -0.8348>}
 calf_OL_r {parent=calf_r, loc=<0 0 -0.0045>, rot=<-0.441 -0.5527 0.441 -0.5527>}
 calf_twist_01_r {parent=calf_r, loc=<1.1013 0 0>, rot=<0 0 0 -1>}
 heel_r {parent=calf_r, loc=<3.8963 0.5585 0>, rot=<0 0 0.2539 -0.9672>}
 foot_r {parent=heel_r, loc=<2.0123 0.0818 0>, rot=<0 0 -0.2262 -0.9741>}
 ball_r {parent=foot_r, loc=<0.9705 -1.2315 -0.0044>, rot=<0.0001 -0.0001 -0.7071 -0.7071>}
 foot_OL_r {parent=foot_r, loc=<0 0 -0.0045>, rot=<0 -0.7071 0 -0.7071>}
 heel_OL_r {parent=heel_r, loc=<0 0 0.0045>, rot=<0.1687 -0.6867 -0.1687 -0.6867>}
 thigh_OL_r {parent=thigh_r, loc=<0 0 -0.0045>, rot=<0 -0.7071 0 -0.7071>}
 thigh_twist_01_r {parent=thigh_r, loc=<0.0451 0 0>, rot=<0 0 0 1>}
 ik_foot_root {parent=root, loc=<0 0 0>, rot=<0 0 0 1>}
 ik_foot_l {parent=ik_foot_root, loc=<1.9393 -4.0033 1.4636>, rot=<-0.0009 -0.7071 0.0009 0.7071>}
 ik_foot_r {parent=ik_foot_root, loc=<-1.9393 -4.0033 1.4636>, rot=<0.7071 -0.0009 -0.7071 0.0009>}
 ik_hand_root {parent=root, loc=<0 0 0>, rot=<0 0 0 1>}
 ik_hand_gun {parent=ik_hand_root, loc=<-2.409 9.3129 1.6809>, rot=<-0.4717 -0.4873 -0.5259 0.5132>}
 ik_hand_l {parent=ik_hand_gun, loc=<-0.1354 4.816 -0.0193>, rot=<0.0281 -0.9996 0.004 0>}
 ik_hand_r {parent=ik_hand_gun, loc=<0 0 0>, rot=<0 0 0 -1>}
]
