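# Post-process AnimBlueprint for the Robo Kitty pet on the shared dog skeleton, adding secondary neck and head motion plus an idle Control Rig.
# When the pet wants anim dynamics, neck_02 and head get either speed-scaled dynamics or, during emotes or DBNO, a stiffer full-strength spring setup; otherwise the input pose passes through.
# The result blends into M_MED_RoboKitty_ControlRig while isInIdle is set, which the update event derives from the main AnimBP's CatIdle curve.
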AnimBlueprint /BRCosmetics/Characters/Pets/Robo_kitty/Meshes/Robo_Kitty_PostP_AnimBP
target_skeleton=/Game/Characters/Pets/Dog_Shared/Dog_Skeleton
graphs [
 AnimGraph {nodes [
   n1 = Control Rig(Source=Use cached pose 'SavedPose8201'()) {
    alphaBoolBlend {blendOption=Linear}
    controlRigAssetReference {blueprintRigClass=&/M_MED_RoboKitty_ControlRig.M_MED_RoboKitty_ControlRig_C}
   }
   Output Pose(Result=Two Way Blend!(A=Use cached pose 'SavedPose8201'(), B=@n1, bAlphaBoolEnabled:bool=Get isInIdle() -> (isInIdle:bool)) -> (Pose) {blendNode {alphaInputType=Bool, alphaBoolBlend {blendInTime=0.3, blendOption=Linear}}}) {layerGroup=None}
   n2 = Get PawnSpeedForAnimDynamics() -> (PawnSpeedForAnimDynamics:float)
   n3 = Anim Dynamics - Bone: neck_02(ComponentPose=Local To Component(LocalPose=Use cached pose 'SavedPose25620'()), Alpha=@n2, ExternalForce="0,0,0") {
    angularDampingOverride=0.9
    angularSpringConstant=200
    bGravityOverrideInSimSpace=true
    bOverrideAngularDamping=true
    gravityScale=0.32
    lODThreshold=1
    numSolverIterationsPostUpdate=4
    numSolverIterationsPreUpdate=16
    simulationSpace=RootRelative
    alphaBoolBlend {blendOption=Linear}
    alphaScaleBiasClamp {bInterpResult=true, bMapRange=true, interpSpeedDecreasing=5, interpSpeedIncreasing=5, inRange {max=500}}
    boundBone {boneName=neck_02}
    componentAppliedLinearAccClamp=<100 100 100>
    componentLinearAccScale=<0.3 0.3 0.3>
    componentLinearVelScale=<0.3 0.3 0.3>
    physicsBodyDefinitions [{
      sphereCollisionRadius=0
      boundBone {boneName=neck_02}
      boxExtents=<3 10 10>
      constraintSetup {angularTargetAxis=AxisY, twistAxis=AxisY, angularLimitsMax=<30 20 20>, angularLimitsMin=<-30 -20 -20>, angularTarget=<0 0 0>}
      localJointOffset=<0 1 0>
     }]
    relativeSpaceBone {boneName=root}
   }
   n4 = Anim Dynamics - Bone: head(ComponentPose=@n3, Alpha=@n2, ExternalForce="0,0,0") {
    angularDampingOverride=0.9
    angularSpringConstant=200
    bGravityOverrideInSimSpace=true
    bOverrideAngularDamping=true
    gravityScale=0.32
    lODThreshold=1
    numSolverIterationsPostUpdate=4
    numSolverIterationsPreUpdate=16
    simulationSpace=RootRelative
    alphaBoolBlend {blendOption=Linear}
    alphaScaleBiasClamp {bInterpResult=true, bMapRange=true, interpSpeedDecreasing=5, interpSpeedIncreasing=5, inRange {max=500}}
    boundBone {boneName=head}
    componentAppliedLinearAccClamp=<100 100 100>
    componentLinearAccScale=<0.3 0.3 0.3>
    componentLinearVelScale=<0.3 0.3 0.3>
    physicsBodyDefinitions [{
      sphereCollisionRadius=0
      boundBone {boneName=head}
      boxExtents=<3 10 10>
      constraintSetup {angularTargetAxis=AxisY, twistAxis=AxisY, angularLimitsMax=<10 40 10>, angularLimitsMin=<-10 -40 -10>, angularTarget=<0 0 0>}
      localJointOffset=<0 1 0>
     }]
    relativeSpaceBone {boneName=root}
   }
   n5 = Component To Local(ComponentPose=@n4)
   n6 = Anim Dynamics - Bone: neck_02(ComponentPose=Local To Component(LocalPose=Use cached pose 'SavedPose25620'()), Alpha=1, ExternalForce="0,0,0") {
    angularDampingOverride=0.99
    angularSpringConstant=1500
    bAngularSpring=true
    bGravityOverrideInSimSpace=true
    bOverrideAngularDamping=true
    gravityScale=0.32
    lODThreshold=1
    numSolverIterationsPostUpdate=2
    numSolverIterationsPreUpdate=8
    simulationSpace=RootRelative
    alphaBoolBlend {blendOption=Linear}
    boundBone {boneName=neck_02}
    componentAppliedLinearAccClamp=<100 100 100>
    componentLinearAccScale=<0.5 0.5 0.5>
    componentLinearVelScale=<0.5 0.5 0.5>
    physicsBodyDefinitions [{
      sphereCollisionRadius=0
      boundBone {boneName=neck_02}
      boxExtents=<5 5 5>
      constraintSetup {angularTargetAxis=AxisY, twistAxis=AxisY, angularLimitsMax=<15 10 2>, angularLimitsMin=<-15 -10 -20>, angularTarget=<0 0 0>}
      localJointOffset=<0 0.25 0>
     }]
    relativeSpaceBone {boneName=root}
   }
   n7 = Anim Dynamics - Bone: head(ComponentPose=@n6, Alpha=1, ExternalForce="0,0,0") {
    angularDampingOverride=0.99
    angularSpringConstant=1500
    bAngularSpring=true
    bGravityOverrideInSimSpace=true
    bOverrideAngularDamping=true
    gravityScale=0.32
    lODThreshold=1
    numSolverIterationsPostUpdate=2
    numSolverIterationsPreUpdate=8
    simulationSpace=RootRelative
    alphaBoolBlend {blendOption=Linear}
    boundBone {boneName=head}
    componentAppliedLinearAccClamp=<100 100 100>
    componentLinearAccScale=<0.5 0.5 0.5>
    componentLinearVelScale=<0.5 0.5 0.5>
    physicsBodyDefinitions [{
      sphereCollisionRadius=0
      boundBone {boneName=head}
      boxExtents=<5 5 5>
      constraintSetup {angularTargetAxis=AxisY, twistAxis=AxisY, angularLimitsMax=<15 10 2>, angularLimitsMin=<-15 -10 -20>, angularTarget=<0 0 0>}
      localJointOffset=<0 0.25 0>
     }]
    relativeSpaceBone {boneName=root}
   }
   n8 = Component To Local(ComponentPose=@n7)
   n9 = Get bIsPlayingEmoteOrDBNO() -> (bIsPlayingEmoteOrDBNO:bool)
   Save cached pose 'SavedPose8201'(Pose=Blend Poses by bool(
     BlendPose_0=Blend Poses by bool(BlendPose_0=@n8, BlendPose_1=@n5, BlendTime_0=1, BlendTime_1=1, bActiveValue=@n9) {blendType=Linear}
     BlendPose_1=Use cached pose 'SavedPose25620'()
     BlendTime_0=1
     BlendTime_1=1
     bActiveValue=Get bPetWantsAnimDynamics() -> (bPetWantsAnimDynamics:bool)
    ) {blendType=Linear}) {cacheName=SavedPose8201}
   AnimGraphNode_SubInput_0 = Input Pose()
   Save cached pose 'SavedPose25620'(Pose=@AnimGraphNode_SubInput_0) {cacheName=SavedPose25620}
  ]}
 EventGraph {nodes [
   Event BlueprintUpdateAnimation()
   > n1 = Cast To FortPlayerPet!(execute, Object=GetOwningActor()) -> (then, CastFailed, AsFort Player Pet:Fort Player Pet Object Reference) {is=title K2Node_DynamicCast}
   > n2 = Cast To B_RoboKitty_AnimBP!(
    execute
    Object=GetAnimInstance(self=Get PetMesh!(self:Fort Player Pet Object Reference=@n1) -> (PetMesh:Skeletal Mesh Component Budgeted Object Reference))
   ) -> (then, CastFailed, AsB Robo Kitty Anim BP:B Robo Kitty Anim BP Object Reference) {is=title K2Node_DynamicCast}
   > Set isInIdle(isInIdle:bool="Not Equal (Float)"!(A:double=GetCurveValue(self=@n2, CurveName=CatIdle), B:double=0) -> (ReturnValue:bool) {is=title K2Node_CallFunction}) -> (Output_Get:bool)
  ]}
]
variables [isInIdle {type=boolean}]
